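# Sound cue for the final demolition impact of book props in interiors.
# A Random node picks between two equally weighted book impact waves, feeding the output at reduced volume.
# It plays through the Fort_Impacts_PSM sound class and references the CONC_Object_Impacts_FourMax concurrency set to cap overlapping impacts.
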SoundCue /Clyde/Sounds/Fort_Impact_Sounds/Demolition/Interior_Props/Demo_Books_ImpactFinal_Cue
root=Random_65
nodes [
 {id=Random_65, children [WavePlayer_212, WavePlayer_215], properties {weights [1,1]}}
 {id=WavePlayer_212, properties {soundWaveAssetPtr=/Clyde/Sounds/Fort_Impact_Sounds/Demolition/Book_ImpactFinal01}}
 {id=WavePlayer_215, properties {soundWaveAssetPtr=/Clyde/Sounds/Fort_Impact_Sounds/Demolition/Book_ImpactFinal04}}
]
properties {
 maxDistance=2097152
 soundClassObject=&Fort_Impacts_PSM
 volumeMultiplier=0.544
 concurrencyOverrides {maxCount=3, resolutionRule=StopOldest}
 concurrencySet [&CONC_Object_Impacts_FourMax]
}
