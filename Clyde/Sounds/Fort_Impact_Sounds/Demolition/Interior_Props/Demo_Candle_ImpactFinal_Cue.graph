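# Sound cue for the final impact of a candle prop, filed among interior demolition impact sounds.
# An equally weighted Random node picks one of three candle impact waves (variants 01, 02 and 04).
# Playback runs at reduced volume through the Fort impacts sound class and references a shared object-impact concurrency set to limit simultaneous voices.
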SoundCue /Clyde/Sounds/Fort_Impact_Sounds/Demolition/Interior_Props/Demo_Candle_ImpactFinal_Cue
root=Random_63
nodes [
 {id=Random_63, children [WavePlayer_203, WavePlayer_204, WavePlayer_205], properties {weights [1,1,1]}}
 {id=WavePlayer_203, properties {soundWaveAssetPtr=/Clyde/Sounds/Fort_Impact_Sounds/Demolition/Candle_ImpactFinal04}}
 {id=WavePlayer_204, properties {soundWaveAssetPtr=/Clyde/Sounds/Fort_Impact_Sounds/Demolition/Candle_ImpactFinal01}}
 {id=WavePlayer_205, properties {soundWaveAssetPtr=/Clyde/Sounds/Fort_Impact_Sounds/Demolition/Candle_ImpactFinal02}}
]
properties {
 maxDistance=2097152
 soundClassObject=&Fort_Impacts_PSM
 volumeMultiplier=0.544
 concurrencyOverrides {maxCount=3, resolutionRule=StopOldest}
 concurrencySet [&CONC_Object_Impacts_FourMax]
}
