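# Sound cue for medium dresser impact hits on interior props in the demolition set of Fort impact sounds.
# A Random node picks one of three dresser drawer hit waves at equal weight, then a Modulator slightly varies pitch on each play.
# It plays in the impacts sound class with small melee weapon attenuation, under an object impact concurrency limit set to prevent new sounds when full.
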SoundCue /Clyde/Sounds/Fort_Impact_Sounds/Demolition/Interior_Props/Demo_Dresser_Med_Impact_Cue
root=Modulator_8
nodes [
 {id=Modulator_8, children [Random_5], properties {pitchMax=1.07, pitchMin=0.92}}
 {id=Random_5, children [WavePlayer_28, WavePlayer_29, WavePlayer_31], properties {weights [1,1,1]}}
 {id=WavePlayer_28, properties {soundWaveAssetPtr=/Clyde/Sounds/Fort_Impact_Sounds/Demolition/Dresser_Drawer_Hit05}}
 {id=WavePlayer_29, properties {soundWaveAssetPtr=/Clyde/Sounds/Fort_Impact_Sounds/Demolition/Dresser_Drawer_Hit01}}
 {id=WavePlayer_31, properties {soundWaveAssetPtr=/Clyde/Sounds/Fort_Impact_Sounds/Demolition/Dresser_Drawer_Hit04}}
]
properties {
 attenuationSettings=&Fort_Weapon_Melee_Sml_Att
 maxDistance=2800
 soundClassObject=&Fort_Impacts
 volumeMultiplier=0.68
 concurrencyOverrides {resolutionRule=PreventNew}
 concurrencySet [&CONC_Object_Impacts_FourMax]
}
