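# Sound cue for the final impact when a harp interior prop is demolished, played in the impacts sound class with a concurrency set for object impacts.
# A quality-level switch sends the first two levels to a layered mix and the last two to a single pre-optimized wave.
# The layered mix combines metal and pitched-up wood impacts with a delayed string hit, followed by a further delayed, quieter, pitched-up string hit for a cascading ring.
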SoundCue /Clyde/Sounds/Fort_Impact_Sounds/Demolition/Interior_Props/Demo_Harp_ImpactFinal_Cue
root=QualityLevel_0
nodes [
 {id=Delay_13, children [Modulator_45], properties {delayMax=0.3, delayMin=0.3}}
 {id=Delay_14, children [Mixer_16], properties {delayMax=0.3, delayMin=0.3}}
 {id=Mixer_15, children [WavePlayer_260, Delay_14, Modulator_46], properties {inputVolume [1,1,1]}}
 {id=Mixer_16, children [WavePlayer_251, Delay_13], properties {inputVolume [1,0.5]}}
 {id=Modulator_0, children [WavePlayer_0]}
 {id=Modulator_44, children [Mixer_15]}
 {id=Modulator_45, children [WavePlayer_252], properties {pitchMax=1.4, pitchMin=1.3}}
 {id=Modulator_46, children [WavePlayer_261], properties {pitchMax=1.3, pitchMin=1.2}}
 {id=QualityLevel_0, children [Modulator_44,Modulator_44,Modulator_0,Modulator_0]}
 {id=WavePlayer_0, properties {soundWaveAssetPtr=/Clyde/Sounds/Fort_Impact_Sounds/Demolition/Harp_ImpactFinal_opt}}
 {id=WavePlayer_251, properties {soundWaveAssetPtr=/Clyde/Sounds/Fort_Impact_Sounds/Demolition/Harp_String_Hit02}}
 {id=WavePlayer_252, properties {soundWaveAssetPtr=/Clyde/Sounds/Fort_Impact_Sounds/Demolition/Harp_String_Hit03}}
 {id=WavePlayer_260, properties {soundWaveAssetPtr=/Clyde/Sounds/Fort_Impact_Sounds/Demolition/Metal_Machine_ImpactFinal01}}
 {id=WavePlayer_261, properties {soundWaveAssetPtr=/Clyde/Sounds/Fort_Impact_Sounds/Demolition/Fort_Impact_WoodFinal_03}}
]
properties {
 maxDistance=2097152
 soundClassObject=&Fort_Impacts_PSM
 volumeMultiplier=0.544
 concurrencyOverrides {maxCount=3, resolutionRule=StopOldest}
 concurrencySet [&CONC_Object_Impacts_FourMax]
}
