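# Sound cue for the impact of a demolition harp prop, routed to the Fort impacts sound class.
# A mixer layers four random picks at differing volumes (small metal, harp string, wood pallet and metal cannon hits), with wood loudest.
# Only the harp string pick allows immediate repeats, and a modulator at default settings processes the mix.
# A small melee weapon attenuation and an object-impact concurrency set govern falloff and voice count.
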SoundCue /Clyde/Sounds/Fort_Impact_Sounds/Demolition/Interior_Props/Demo_Harp_Impact_Cue
root=Modulator_43
nodes [
 {id=Mixer_14, children [Random_78,Random_76,Random_79,Random_77], properties {inputVolume [0.8,0.7,1,0.75]}}
 {id=Modulator_43, children [Mixer_14]}
 {id=Random_76, children [WavePlayer_242, WavePlayer_243, WavePlayer_250], properties {bRandomizeWithoutReplacement=false, weights [1,1,1]}}
 {id=Random_77, children [WavePlayer_244, WavePlayer_245], properties {weights [1,1]}}
 {id=Random_78, children [WavePlayer_246, WavePlayer_247], properties {weights [1,1]}}
 {id=Random_79, children [WavePlayer_248, WavePlayer_249], properties {weights [1,1]}}
 {id=WavePlayer_242, properties {soundWaveAssetPtr=/Clyde/Sounds/Fort_Impact_Sounds/Demolition/Harp_String_Hit02}}
 {id=WavePlayer_243, properties {soundWaveAssetPtr=/Clyde/Sounds/Fort_Impact_Sounds/Demolition/Harp_String_Hit03}}
 {id=WavePlayer_244, properties {soundWaveAssetPtr=/Clyde/Sounds/Fort_Impact_Sounds/Demolition/Metal_Cannon_Impact02}}
 {id=WavePlayer_245, properties {soundWaveAssetPtr=/Clyde/Sounds/Fort_Impact_Sounds/Demolition/Metal_Cannon_Impact01}}
 {id=WavePlayer_246, properties {soundWaveAssetPtr=/Clyde/Sounds/Fort_Impact_Sounds/Demolition/Metal_Small_Impact04}}
 {id=WavePlayer_247, properties {soundWaveAssetPtr=/Clyde/Sounds/Fort_Impact_Sounds/Demolition/Metal_Small_Impact03}}
 {id=WavePlayer_248, properties {soundWaveAssetPtr=/Clyde/Sounds/Fort_Impact_Sounds/Demolition/Wood_Pallets_Impact03}}
 {id=WavePlayer_249, properties {soundWaveAssetPtr=/Clyde/Sounds/Fort_Impact_Sounds/Demolition/Wood_Pallets_Impact04}}
 {id=WavePlayer_250, properties {soundWaveAssetPtr=/Clyde/Sounds/Fort_Impact_Sounds/Demolition/Harp_String_Hit01}}
]
properties {
 attenuationSettings=&Fort_Weapon_Melee_Sml_Att
 maxDistance=2800
 soundClassObject=&Fort_Impacts
 volumeMultiplier=0.68
 concurrencyOverrides {maxCount=3, resolutionRule=StopOldest}
 concurrencySet [&CONC_Object_Impacts_FourMax]
}
