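# Sound cue for demolition impacts on interior locker props.
# A random node picks one of three equally weighted locker impact waves, then a modulator slightly varies its pitch.
# It plays in the impacts sound class with small melee weapon attenuation and reduced volume.
# It references a four-max object impact concurrency set and also carries a prevent-new concurrency override.
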SoundCue /Clyde/Sounds/Fort_Impact_Sounds/Demolition/Interior_Props/Demo_Locker_Impact_Cue
root=Modulator_10
nodes [
 {id=Modulator_10, children [Random_24], properties {pitchMax=1.07, pitchMin=0.92}}
 {id=Random_24, children [WavePlayer_85, WavePlayer_86, WavePlayer_87], properties {weights [1,1,1]}}
 {id=WavePlayer_85, properties {soundWaveAssetPtr=/Clyde/Sounds/Fort_Impact_Sounds/Demolition/Locker_Impact01}}
 {id=WavePlayer_86, properties {soundWaveAssetPtr=/Clyde/Sounds/Fort_Impact_Sounds/Demolition/Locker_Impact02}}
 {id=WavePlayer_87, properties {soundWaveAssetPtr=/Clyde/Sounds/Fort_Impact_Sounds/Demolition/Locker_Impact03}}
]
properties {
 attenuationSettings=&Fort_Weapon_Melee_Sml_Att
 maxDistance=2800
 soundClassObject=&Fort_Impacts
 volumeMultiplier=0.68
 concurrencyOverrides {resolutionRule=PreventNew}
 concurrencySet [&CONC_Object_Impacts_FourMax]
}
