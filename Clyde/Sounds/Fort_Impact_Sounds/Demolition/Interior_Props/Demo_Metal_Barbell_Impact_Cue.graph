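# Sound cue for a metal barbell impact, part of the demolition interior prop impact sounds.
# A random node picks one of three equally weighted gym barbell recordings, then a modulator slightly varies its pitch on each play.
# The cue plays at reduced volume in the impacts sound class, with small-melee-weapon attenuation and a shared object-impact concurrency set.
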SoundCue /Clyde/Sounds/Fort_Impact_Sounds/Demolition/Interior_Props/Demo_Metal_Barbell_Impact_Cue
root=Modulator_2
nodes [
 {id=Modulator_2, children [Random_27], properties {pitchMax=1.07, pitchMin=0.92}}
 {id=Random_27, children [WavePlayer_78, WavePlayer_79, WavePlayer_80], properties {weights [1,1,1]}}
 {id=WavePlayer_78, properties {soundWaveAssetPtr=/Clyde/Sounds/Fort_Impact_Sounds/Demolition/Gym_Barbell_Impact04}}
 {id=WavePlayer_79, properties {soundWaveAssetPtr=/Clyde/Sounds/Fort_Impact_Sounds/Demolition/Gym_Barbell_Impact01}}
 {id=WavePlayer_80, properties {soundWaveAssetPtr=/Clyde/Sounds/Fort_Impact_Sounds/Demolition/Gym_Barbell_Impact02}}
]
properties {
 attenuationSettings=&Fort_Weapon_Melee_Sml_Att
 maxDistance=2800
 soundClassObject=&Fort_Impacts
 volumeMultiplier=0.34
 concurrencyOverrides {maxCount=3, resolutionRule=StopOldest}
 concurrencySet [&CONC_Object_Impacts_FourMax]
}
