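# Sound cue for the final impact of demolishing a medium metal computer interior prop.
# A single computer final-impact wave feeds a Modulator that randomizes pitch from normal to slightly higher for variation.
# The cue plays in the Fort impacts sound class at reduced volume and references a shared object-impact concurrency set, alongside stop-oldest override settings.
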SoundCue /Clyde/Sounds/Fort_Impact_Sounds/Demolition/Interior_Props/Demo_Metal_ComputerMedium_ImpactFinal_Cue
root=Modulator_0
nodes [
 {id=Modulator_0, children [WavePlayer_3], properties {pitchMax=1.2, pitchMin=1}}
 {id=WavePlayer_3, properties {soundWaveAssetPtr=&/Computer_FinalImpact_01}}
]
properties {
 maxDistance=2097152
 soundClassObject=&Fort_Impacts_PSM
 volumeMultiplier=0.544
 concurrencyOverrides {maxCount=3, resolutionRule=StopOldest}
 concurrencySet [&CONC_Object_Impacts_FourMax]
}
