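# Sound cue for impacts on metal computer server interior props, filed under demolition impact sounds.
# A Random node picks one of three computer impact waves with equal weight, feeding a Modulator left at default settings.
# It plays in the Fort_Impacts sound class at reduced volume, using small melee weapon attenuation.
# An object impact concurrency set is assigned, alongside override values that stop the oldest sound.
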SoundCue /Clyde/Sounds/Fort_Impact_Sounds/Demolition/Interior_Props/Demo_Metal_ComputerServer_Impact_Cue
root=Modulator_0
nodes [
 {id=Modulator_0, children [Random_3]}
 {id=Random_3, children [WavePlayer_1,WavePlayer_2,WavePlayer_3], properties {weights [1,1,1]}}
 {id=WavePlayer_1, properties {soundWaveAssetPtr=&/Computer_Impact_01}}
 {id=WavePlayer_2, properties {soundWaveAssetPtr=&/Computer_Impact_02}}
 {id=WavePlayer_3, properties {soundWaveAssetPtr=&/Computer_Impact_03}}
]
properties {
 attenuationSettings=&Fort_Weapon_Melee_Sml_Att
 maxDistance=2800
 soundClassObject=&Fort_Impacts
 volumeMultiplier=0.544
 concurrencyOverrides {maxCount=3, resolutionRule=StopOldest}
 concurrencySet [&CONC_Object_Impacts_FourMax]
}
