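# Sound cue for impacts on small metal computer props in interior demolition.
# It randomly picks one of three equally weighted computer impact waves, then a modulator raises its pitch by a varying amount.
# Small-melee weapon attenuation, the impacts sound class and an object-impact concurrency set govern playback, alongside a stop-oldest override at reduced volume.
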SoundCue /Clyde/Sounds/Fort_Impact_Sounds/Demolition/Interior_Props/Demo_Metal_ComputerSmall_Impact_Cue
root=Modulator_2
nodes [
 {id=Modulator_2, children [Random_5], properties {pitchMax=1.5, pitchMin=1.3}}
 {id=Random_5, children [WavePlayer_13, WavePlayer_14, WavePlayer_15], properties {weights [1,1,1]}}
 {id=WavePlayer_13, properties {soundWaveAssetPtr=&/Computer_Impact_01}}
 {id=WavePlayer_14, properties {soundWaveAssetPtr=&/Computer_Impact_02}}
 {id=WavePlayer_15, properties {soundWaveAssetPtr=&/Computer_Impact_03}}
]
properties {
 attenuationSettings=&Fort_Weapon_Melee_Sml_Att
 maxDistance=2800
 soundClassObject=&Fort_Impacts
 volumeMultiplier=0.544
 concurrencyOverrides {maxCount=3, resolutionRule=StopOldest}
 concurrencySet [&CONC_Object_Impacts_FourMax]
}
