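# Sound cue for the final demolition impact of a metal pirate loot bag prop.
# A mixer layers one of two small metal death sounds at half volume with one of two explosive impact layers at full volume.
# A root modulator randomizes pitch slightly above normal for variation.
# The cue plays at reduced volume in the Fort impacts sound class and references a shared object-impact concurrency set.
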SoundCue /Clyde/Sounds/Fort_Impact_Sounds/Demolition/Interior_Props/Demo_Metal_PirateLootBag_ImpactFinal_Cue
root=Modulator_6
nodes [
 {id=Mixer_3, children [Random_10,Random_19], properties {inputVolume [0.5,1]}}
 {id=Modulator_6, children [Mixer_3], properties {pitchMax=1.2, pitchMin=1.1}}
 {id=Random_10, children [WavePlayer_28, WavePlayer_29], properties {weights [1,1]}}
 {id=Random_19, children [WavePlayer_59, WavePlayer_60], properties {weights [1,1]}}
 {id=WavePlayer_28, properties {soundWaveAssetPtr=/Clyde/Sounds/Fort_Impact_Sounds/Demolition/Metal_Small_Death02}}
 {id=WavePlayer_29, properties {soundWaveAssetPtr=/Clyde/Sounds/Fort_Impact_Sounds/Demolition/Metal_Small_Death01}}
 {id=WavePlayer_59, properties {soundWaveAssetPtr=/Clyde/Sounds/Fort_Impact_Sounds/Demolition/Explo_Impact_Layer01}}
 {id=WavePlayer_60, properties {soundWaveAssetPtr=/Clyde/Sounds/Fort_Impact_Sounds/Demolition/Explo_Impact_Layer02}}
]
properties {
 maxDistance=2097152
 soundClassObject=&Fort_Impacts_PSM
 volumeMultiplier=0.544
 concurrencyOverrides {maxCount=3, resolutionRule=StopOldest}
 concurrencySet [&CONC_Object_Impacts_FourMax]
}
