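# Sound cue for the destruction of a pipe organ prop, played as a demolition impact under the Fort_Impacts sound class.
# A quality-level root selects between two identical layered mixes and two single pre-made "opt" waves with pitch modulation.
# Each layered mix blends wood, barbell and metal railing sounds with randomized pitch, the barbell pitched lower and briefly delayed.
# An object-impact concurrency set is assigned alongside a PreventNew override.
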SoundCue /Clyde/Sounds/Fort_Impact_Sounds/Demolition/Interior_Props/Demo_PipeOrgan_Destroyed_Cue
root=QualityLevel_0
nodes [
 {id=Delay_0, children [WavePlayer_1], properties {delayMax=0.25, delayMin=0.15}}
 {id=Delay_10, children [WavePlayer_169], properties {delayMax=0.25, delayMin=0.15}}
 {id=Mixer_0, children [Modulator_0,Modulator_1,Modulator_2], properties {inputVolume [1,0.9,0.8]}}
 {id=Mixer_8, children [Modulator_32,Modulator_30,Modulator_33], properties {inputVolume [1,0.9,0.8]}}
 {id=Modulator_0, children [WavePlayer_3], properties {pitchMax=1, pitchMin=0.85}}
 {id=Modulator_1, children [Delay_0], properties {pitchMax=0.7, pitchMin=0.6}}
 {id=Modulator_2, children [WavePlayer_2], properties {pitchMax=1, pitchMin=0.85}}
 {id=Modulator_3, children [WavePlayer_0], properties {pitchMin=0.9}}
 {id=Modulator_30, children [Delay_10], properties {pitchMax=0.7, pitchMin=0.6}}
 {id=Modulator_32, children [WavePlayer_199], properties {pitchMax=1, pitchMin=0.85}}
 {id=Modulator_33, children [WavePlayer_197], properties {pitchMax=1, pitchMin=0.85}}
 {id=Modulator_4, children [WavePlayer_4], properties {pitchMin=0.9}}
 {id=QualityLevel_0, children [Mixer_8,Mixer_0,Modulator_3,Modulator_4]}
 {id=WavePlayer_0, properties {soundWaveAssetPtr=&/PipeOrgan_Destroyed_opt}}
 {id=WavePlayer_1, properties {soundWaveAssetPtr=/Clyde/Sounds/Fort_Impact_Sounds/Demolition/Gym_Barbell_ImpactFinal04}}
 {id=WavePlayer_169, properties {soundWaveAssetPtr=&Gym_Barbell_ImpactFinal04}}
 {id=WavePlayer_197, properties {soundWaveAssetPtr=/Clyde/Sounds/Fort_Impact_Sounds/Demolition/Metal_Railing_Death02}}
 {id=WavePlayer_199, properties {soundWaveAssetPtr=/Clyde/Sounds/Fort_Impact_Sounds/Demolition/Fort_Impact_WoodFinal_03}}
 {id=WavePlayer_2, properties {soundWaveAssetPtr=&Metal_Railing_Death02}}
 {id=WavePlayer_3, properties {soundWaveAssetPtr=&Fort_Impact_WoodFinal_03}}
 {id=WavePlayer_4, properties {soundWaveAssetPtr=&PipeOrgan_Destroyed_opt}}
]
properties {
 maxDistance=2097152
 soundClassObject=&Fort_Impacts
 volumeMultiplier=0.476
 concurrencyOverrides {resolutionRule=PreventNew}
 concurrencySet [&CONC_Object_Impacts_FourMax]
}
