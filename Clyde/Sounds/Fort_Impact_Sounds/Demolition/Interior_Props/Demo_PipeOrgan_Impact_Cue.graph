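# Sound cue for a pipe organ prop impact in the demolition interior-props set, classed as a Fort impact sound.
# A quality-level root sends its first two slots to a modulated random pick between two impact waves, and its last two to the first wave alone with a wider pitch range.
# Small melee-weapon attenuation shapes falloff, and a four-max object-impact concurrency set limits simultaneous plays.
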SoundCue /Clyde/Sounds/Fort_Impact_Sounds/Demolition/Interior_Props/Demo_PipeOrgan_Impact_Cue
root=QualityLevel_0
nodes [
 {id=Modulator_0, children [Random_0]}
 {id=Modulator_1, children [WavePlayer_2], properties {pitchMax=1.1, pitchMin=0.9}}
 {id=QualityLevel_0, children [Modulator_0,Modulator_0,Modulator_1,Modulator_1]}
 {id=Random_0, children [WavePlayer_0,WavePlayer_1], properties {weights [1,1]}}
 {id=WavePlayer_0, properties {soundWaveAssetPtr=/Clyde/Sounds/Fort_Impact_Sounds/Demolition/PipeOrgan_Impact_01}}
 {id=WavePlayer_1, properties {soundWaveAssetPtr=/Clyde/Sounds/Fort_Impact_Sounds/Demolition/PipeOrgan_Impact_02}}
 {id=WavePlayer_2, properties {soundWaveAssetPtr=&PipeOrgan_Impact_01}}
]
properties {
 attenuationSettings=&Fort_Weapon_Melee_Sml_Att
 maxDistance=2800
 soundClassObject=&Fort_Impacts
 volumeMultiplier=0.476
 concurrencyOverrides {resolutionRule=PreventNew}
 concurrencySet [&CONC_Object_Impacts_FourMax]
}
