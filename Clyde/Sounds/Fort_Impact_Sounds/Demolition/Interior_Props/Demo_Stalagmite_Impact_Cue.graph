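# Sound cue for a stalagmite impact, filed under demolition sounds for interior props.
# An equally weighted Random node picks one of four stalagmite impact waves, then a Modulator varies it for less repetition.
# It plays in the Fort_Impacts sound class with small melee weapon attenuation and slightly reduced volume.
# The shared CONC_Object_Impacts_FourMax concurrency set limits how many impacts overlap.
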SoundCue /Clyde/Sounds/Fort_Impact_Sounds/Demolition/Interior_Props/Demo_Stalagmite_Impact_Cue
root=Modulator_37
nodes [
 {id=Modulator_37, children [Random_0]}
 {id=Random_0, children [WavePlayer_0,WavePlayer_1,WavePlayer_2,WavePlayer_3], properties {weights [1,1,1,1]}}
 {id=WavePlayer_0, properties {soundWaveAssetPtr=&/stalagmite_impact_01}}
 {id=WavePlayer_1, properties {soundWaveAssetPtr=&/stalagmite_impact_02}}
 {id=WavePlayer_2, properties {soundWaveAssetPtr=&/stalagmite_impact_03}}
 {id=WavePlayer_3, properties {soundWaveAssetPtr=&/stalagmite_impact_04}}
]
properties {
 attenuationSettings=&Fort_Weapon_Melee_Sml_Att
 maxDistance=2800
 soundClassObject=&Fort_Impacts
 volumeMultiplier=0.68
 concurrencyOverrides {maxCount=3, resolutionRule=StopOldest}
 concurrencySet [&CONC_Object_Impacts_FourMax]
}
