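# Sound cue for toolbox impact hits in the demolition interior props set, classed under Fort_Impacts with small melee weapon attenuation.
# A quality-level root selects one of four branches with a shrinking wave pool: four, three, two, then a single Toolbox_Hits sound.
# Multi-wave branches pick at equal weight, and every branch adds pitch variation.
# The CONC_Object_Impacts_FourMax concurrency set governs simultaneous plays.
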SoundCue /Clyde/Sounds/Fort_Impact_Sounds/Demolition/Interior_Props/Demo_Toolbox_Impact_Cue
root=QualityLevel_0
nodes [
 {id=Modulator_0, children [Random_0], properties {pitchMax=1.07, pitchMin=0.92}}
 {id=Modulator_1, children [Random_1], properties {pitchMax=1.07, pitchMin=0.92}}
 {id=Modulator_10, children [Random_9], properties {pitchMax=1.07, pitchMin=0.92}}
 {id=Modulator_2, children [WavePlayer_5], properties {pitchMax=1.1, pitchMin=0.9}}
 {id=QualityLevel_0, children [Modulator_10,Modulator_0,Modulator_1,Modulator_2]}
 {id=Random_0, children [WavePlayer_0,WavePlayer_1,WavePlayer_2], properties {weights [1,1,1]}}
 {id=Random_1, children [WavePlayer_3,WavePlayer_4], properties {weights [1,1]}}
 {id=Random_9, children [WavePlayer_35, WavePlayer_36, WavePlayer_37, WavePlayer_40], properties {weights [1,1,1,1]}}
 {id=WavePlayer_0, properties {soundWaveAssetPtr=/Clyde/Sounds/Fort_Impact_Sounds/Demolition/Toolbox_Hits8}}
 {id=WavePlayer_1, properties {soundWaveAssetPtr=/Clyde/Sounds/Fort_Impact_Sounds/Demolition/Toolbox_Hits1}}
 {id=WavePlayer_2, properties {soundWaveAssetPtr=/Clyde/Sounds/Fort_Impact_Sounds/Demolition/Toolbox_Hits2}}
 {id=WavePlayer_3, properties {soundWaveAssetPtr=&Toolbox_Hits8}}
 {id=WavePlayer_35, properties {soundWaveAssetPtr=&Toolbox_Hits8}}
 {id=WavePlayer_36, properties {soundWaveAssetPtr=&Toolbox_Hits1}}
 {id=WavePlayer_37, properties {soundWaveAssetPtr=&Toolbox_Hits2}}
 {id=WavePlayer_4, properties {soundWaveAssetPtr=&Toolbox_Hits1}}
 {id=WavePlayer_40, properties {soundWaveAssetPtr=/Clyde/Sounds/Fort_Impact_Sounds/Demolition/Toolbox_Hits5}}
 {id=WavePlayer_5, properties {soundWaveAssetPtr=&Toolbox_Hits8}}
]
properties {
 attenuationSettings=&Fort_Weapon_Melee_Sml_Att
 maxDistance=2800
 soundClassObject=&Fort_Impacts
 volumeMultiplier=0.68
 concurrencyOverrides {resolutionRule=PreventNew}
 concurrencySet [&CONC_Object_Impacts_FourMax]
}
