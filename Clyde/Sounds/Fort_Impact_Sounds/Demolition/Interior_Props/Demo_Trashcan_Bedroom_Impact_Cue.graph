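# Sound cue for impacts on a bedroom trashcan prop, part of the interior-prop demolition impact sounds.
# A random node picks one of three equally weighted wastebasket hit waves, and a modulator above it slightly varies the pitch.
# It plays in the impacts sound class at reduced volume, using small melee-weapon attenuation and an object-impact concurrency set.
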SoundCue /Clyde/Sounds/Fort_Impact_Sounds/Demolition/Interior_Props/Demo_Trashcan_Bedroom_Impact_Cue
root=Modulator_13
nodes [
 {id=Modulator_13, children [Random_0], properties {pitchMax=1.07, pitchMin=0.92}}
 {id=Random_0, children [WavePlayer_0,WavePlayer_1,WavePlayer_2], properties {weights [1,1,1]}}
 {id=WavePlayer_0, properties {soundWaveAssetPtr=/Clyde/Sounds/Fort_Impact_Sounds/Demolition/WasteBasket_Bedroom_Hit11}}
 {id=WavePlayer_1, properties {soundWaveAssetPtr=/Clyde/Sounds/Fort_Impact_Sounds/Demolition/WasteBasket_Bedroom_Hit04}}
 {id=WavePlayer_2, properties {soundWaveAssetPtr=/Clyde/Sounds/Fort_Impact_Sounds/Demolition/WasteBasket_Bedroom_Hit05}}
]
properties {
 attenuationSettings=&Fort_Weapon_Melee_Sml_Att
 maxDistance=2800
 soundClassObject=&Fort_Impacts
 volumeMultiplier=0.68
 concurrencyOverrides {resolutionRule=PreventNew}
 concurrencySet [&CONC_Object_Impacts_FourMax]
}
