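# Sound cue for impacts on workbench junk interior props, in the demolition set of Fort impact sounds.
# A random node picks one of three equally weighted hit waves, and a modulator adds slight pitch variation.
# It plays in the impacts sound class with small melee weapon attenuation and reduced volume.
# A four-max object impact concurrency set is assigned, alongside a prevent-new override setting.
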SoundCue /Clyde/Sounds/Fort_Impact_Sounds/Demolition/Interior_Props/Demo_Workbench_Junk_Impact_Cue
root=Modulator_10
nodes [
 {id=Modulator_10, children [Random_8], properties {pitchMax=1.07, pitchMin=0.92}}
 {id=Random_8, children [WavePlayer_32, WavePlayer_33, WavePlayer_34], properties {weights [1,1,1]}}
 {id=WavePlayer_32, properties {soundWaveAssetPtr=/Clyde/Sounds/Fort_Impact_Sounds/Demolition/Workbench_Junk_Hit3}}
 {id=WavePlayer_33, properties {soundWaveAssetPtr=/Clyde/Sounds/Fort_Impact_Sounds/Demolition/Workbench_Junk_Hit1}}
 {id=WavePlayer_34, properties {soundWaveAssetPtr=/Clyde/Sounds/Fort_Impact_Sounds/Demolition/Workbench_Junk_Hit2}}
]
properties {
 attenuationSettings=&Fort_Weapon_Melee_Sml_Att
 maxDistance=2800
 soundClassObject=&Fort_Impacts
 volumeMultiplier=0.68
 concurrencyOverrides {resolutionRule=PreventNew}
 concurrencySet [&CONC_Object_Impacts_FourMax]
}
